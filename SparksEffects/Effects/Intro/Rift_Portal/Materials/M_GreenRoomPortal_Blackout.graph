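# Translucent, unlit, two-sided Niagara sprite material, named as a blackout layer for the Green Room rift portal intro effect.
# Radial UVs are spiraled to pan a caustics texture, which distorts a second, sine-warped, mirrored caustics sample.
# Circular sphere masks shape the moving alpha and cyan hot-edge glow, and particle color tints the emissive.
# A Reveal dynamic parameter, plus a test scalar, is power-curved into a shared reveal value.
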Material /SparksEffects/Effects/Intro/Rift_Portal/Materials/M_GreenRoomPortal_Blackout
graph {
 properties {
  bUsedWithNiagaraSprites=true
  blendMode=BLEND_Translucent
  pixelDepthOffsetMode=PDOM_Legacy
  shadingModel=MSM_Unlit
  twoSided=true
  usedShadingModels=Unlit
 }
 nodes [
  n1 = TextureCoordinate()
  n2 = Constant2Vector() {g=0.5, r=0.5}
  n3 = Multiply(A=SphereMask(A=@n1, B=@n2) {attenuationRadius=0.15, hardnessPercent=0.25}) {constB=4}
  n4 = SphereMask(A=@n1, B=@n2) {attenuationRadius=0.25, hardnessPercent=0.01}
  n5 = Multiply(A=@n4) {constB=2}
  Add(A=@n3, B=@n5)
  n6 = MaterialFunctionCall() {materialFunction=&VectorToRadialValue}
  n7 = ComponentMask(None=@n6.Radial Coordinates) {g=true}
  n8 = MaterialFunctionCall(Float2=AppendVector(A=ComponentMask(None=@n6.Radial Coordinates) {r=true}, B=OneMinus(None=@n7))) {materialFunction=&BreakOutFloat2Components}
  n9 = LinearInterpolate(A=@n8.G, B=Multiply(A=@n8.G, B=@n8.G), Alpha=ScalarParameter() {parameterName="UV Bias (Bump)"})
  n10 = Add(A=@n8.R, B=Multiply(A=@n9, B=ScalarParameter() {defaultValue=0.1, parameterName=Spiral U}))
  n11 = Divide(B=ScalarParameter() {defaultValue=1, parameterName=Tile V for Spiral})
  n12 = ScalarParameter() {parameterName="Spiral V (Whole Numbers)"}
  n13 = Add(A=@n9, B=Multiply(A=@n8.R, B=Multiply(A=@n11, B=@n12)))
  n14 = Time()
  n15 = TextureCoordinate()
  n16 = TextureSample!(
   UVs=MaterialFunctionCall!(
    UV=AppendVector(A=@n10, B=@n13)
    Tile U=ScalarParameter() {defaultValue=-2, parameterName="Tile U (S)_1"}
    Tile V=ScalarParameter() {defaultValue=0.5, parameterName="Tile V (S)_1"}
    Pan U=ScalarParameter() {defaultValue=0.1, parameterName="Pan U (S)_1"}
    Pan V=ScalarParameter() {defaultValue=-0.5, parameterName="Pan V (S)_1"}
    Offset U
    Offset V
    Time=Multiply(A=@n14) {constB=0.25}
   ) {materialFunction=/Game/Effects/Fort_Effects/Materials/_Functions/MF_PanTileOffset}.UV's
   Tex
   "DDX(UVs)"=DDX(Value=@n15)
   "DDY(UVs)"=DDY(Value=@n15)
   Apply View MipBias
  ) {mipValueMode=TMVM_Derivative, texture=/Game/Effects/Fort_Effects/Textures/Tile/T_ShieldCaustics_Tile_Chromatic}
  n17 = Multiply(A=@n4, B=@n16.G)
  n18 = SphereMask(A=@n1, B=@n2) {attenuationRadius=0.2, hardnessPercent=0.001}
  n19 = Power(Base=@n5, Exp=ScalarParameter() {defaultValue=4, parameterName=RevealEdgeAmt})
  LinearInterpolate(A=@n17, B=@n18, Alpha=@n19)
  n20 = ParticleColor()
  n21 = NamedRerouteUsage()
  LinearInterpolate(A=@n20.RGB, B=NamedRerouteUsage(), Alpha=Saturate(None=@n21))
  LinearInterpolate(Alpha=NamedRerouteUsage()) {constA=0.25}
  n22 = NamedRerouteUsage()
  LinearInterpolate(
   A=ScalarParameter() {defaultValue=0.15, parameterName=RevealRadiusLow}
   B=ScalarParameter() {defaultValue=0.15, parameterName=RevealRadiusHigh}
   Alpha=@n22
  )
  LinearInterpolate(
   A=ScalarParameter() {defaultValue=0.25, parameterName=RevealHardnessLow}
   B=ScalarParameter() {defaultValue=0.25, parameterName=RevealHardnessHigh}
   Alpha=@n22
  )
  n23 = ScalarParameter() {defaultValue=1, parameterName="Tile U (S)"}
  n24 = ScalarParameter() {defaultValue=1, parameterName="Tile V (S)"}
  n25 = ScalarParameter() {defaultValue=-0.25, parameterName="Pan U (S)"}
  n26 = ScalarParameter() {defaultValue=0.1, parameterName="Pan V (S)"}
  n27 = Multiply(A=Time()) {constB=-0.25}
  n28 = TextureCoordinate()
  n29 = ComponentMask(None=@n28) {r=true}
  n30 = ComponentMask(None=@n28) {g=true}
  n31 = Sine(None=Add(A=@n30) {constB=-0.5})
  n32 = LinearInterpolate(A=ScalarParameter() {defaultValue=0.05, parameterName=SineAmt}, Alpha=NamedRerouteUsage()) {constA=0.15, constB=0}
  n33 = AppendVector(A=Add(A=Add(A=@n29) {constB=-0.5}, B=Multiply(A=@n31, B=@n32)), B=@n30)
  n34 = ComponentMask(None=@n16.RGB) {g=true, r=true}
  n35 = Abs(None=Add(A=@n33, B=ConstantBiasScale(None=@n34) {bias=-0.5, scale=0.2}))
  n36 = TextureCoordinate()
  n37 = LinearInterpolate(
   A=TextureSample(
    UVs=MaterialFunctionCall(UV=@n35, Tile U=@n23, Tile V=@n24, Pan U=@n25, Pan V=@n26, Time=@n27) {materialFunction=&MF_PanTileOffset}.UV's
    "DDX(UVs)"=DDX(Value=@n36)
    "DDY(UVs)"=DDY(Value=@n36)
   ) {mipValueMode=TMVM_Derivative, texture=&T_ShieldCaustics_Tile_Chromatic}.RGB
   B=@n16.RGB
   Alpha=NamedRerouteUsage()
  )
  n38 = Sine(None=@n4)
  n39 = Multiply(A=@n38, B=Constant3Vector() {constant=<rgba 0 1.0668 1.5 1>})
  n40 = Saturate(None=Multiply(A=@n38) {constB=0.25})
  MaterialFunctionCall(Emissive=Multiply(A=LinearInterpolate(A=@n37, B=@n39, Alpha=@n40) {constAlpha=0.15}, B=@n20.RGB)) {materialFunction=&MF_FX_EmissiveAdjustments}
  MaterialFunctionCall() {materialFunction=&DebugTimeSine}
  n41 = Sine(None=@n18)
  n42 = SmoothStep(Value=LinearInterpolate(A=@n17, B=Saturate(None=@n5), Alpha=@n19)) {constMax=0.25, desc=Moving alpha mask}
  Multiply(A=@n41, B=@n42)
  n43 = Add(
   A=DynamicParameter() {defaultValue=<rgba 0 1 1 1>, paramNames [Reveal,Param2,Param3,Param4]}.Reveal
   B=ScalarParameter() {parameterName=RevealTest}
  )
  NamedRerouteDeclaration(None=Power(Base=Saturate(None=@n43), Exp=ScalarParameter() {defaultValue=4, parameterName=RevealPow})) {name=RevealValue, nodeColor=<rgba 0.0781 1 0 1>}
  NamedRerouteDeclaration(None=Saturate(None=@n8.G)) {name=Round Mask, nodeColor=<rgba 1 0.7031 0 1>}
  n44 = Saturate(None=NamedRerouteUsage())
  n45 = NamedRerouteUsage()
  n46 = Multiply(A=Power(Base=@n18) {constExponent=0.05}) {constB=2}
  n47 = Multiply(A=Add(A=@n17, B=Multiply(A=@n44, B=@n45)), B=OneMinus(None=Saturate(None=@n46)))
  NamedRerouteDeclaration(None=Multiply(A=Constant3Vector() {constant=<rgba 0 2.5894 3 1>}, B=NamedRerouteDeclaration(None=@n47) {name=HotEdgeMask, nodeColor=<rgba 0.4297 1 0 1>})) {name=HotEdgeGlow, nodeColor=<rgba 1 0.3281 0 1>}
  OneMinus(None=@n3)
  OneMinus(None=Multiply(A=@n21) {constB=3})
  n48 = Multiply(A=Abs(None=ComponentMask(None=@n33) {r=true})) {constB=2}
  Power(Base=Saturate(None=OneMinus(None=@n48)), Exp=LinearInterpolate(Alpha=NamedRerouteUsage()) {constA=32})
  Reroute(None=@n35)
  n49 = NamedRerouteUsage()
  Saturate(None=Add(A=@n42, B=Saturate(None=@n49)))
  n50 = ParticleColor()
  Saturate(None=Multiply(A=Multiply(A=@n50.A, B=@n42), B=LinearInterpolate(Alpha=@n38) {constA=0.95}))
  SmoothStep(Value=@n41) {constMax=0.25, desc=Moving alpha mask}
 ]
}
