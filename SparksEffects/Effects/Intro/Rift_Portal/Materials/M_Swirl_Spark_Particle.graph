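# Translucent, unlit, two-sided Niagara sprite material for swirling plasma sparks in the Rift Portal intro effect.
# Bulged, swirled and turn-distorted UVs sample one of two plasma masks, picked randomly per particle.
# Dynamic parameters choose half or full mask, drive alpha erosion, and control a sine-based radial reveal inside a sphere mask.
# Particle color tints the boosted emissive, and particle alpha scales an opacity chain with randomly offset flicker.
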Material /SparksEffects/Effects/Intro/Rift_Portal/Materials/M_Swirl_Spark_Particle
graph {
 properties {
  bUsedWithNiagaraSprites=true
  blendMode=BLEND_Translucent
  pixelDepthOffsetMode=PDOM_Legacy
  shadingModel=MSM_Unlit
  twoSided=true
  usedShadingModels=Unlit
 }
 nodes [
  n1 = TextureCoordinate()
  Add(A=@n1) {constB=-0.75}
  n2 = TextureCoordinate()
  n3 = Time()
  Add(A=@n2, B=AppendVector(A=Constant(), B=Multiply(A=@n3) {constB=-1}))
  Constant() {r=0.1}
  Constant()
  n4 = TextureCoordinate()
  DDX(Value=@n4)
  DDY(Value=@n4)
  LinearInterpolate(Alpha=ParticleRandom()) {constA=-1, constB=-3}
  n5 = ParticleColor()
  n6 = MaterialFunctionCall!(
   UV TexCoord=MaterialFunctionCall!(UV, Bulge Intensity XY=Constant2Vector() {g=-1, r=-1}) {materialFunction=/Game/Effects/Niagara/MasterMaterials/Functions/Mojo_MF/MF_FX_Bulge_UV}.Out UV
   Time
   Speed=Constant() {r=-0.0015}
   Spin=LinearInterpolate(B=Constant() {r=0.5}, Alpha=ParticleRandom()) {constA=0.25}
  ) {materialFunction=/Game/Effects/Niagara/MasterMaterials/Functions/Mojo_MF/MF_FX_Swirl_UV}
  n7 = OneMinus(None=ComponentMask(None=@n6.Out UV TexCoord) {g=true})
  n8 = MaterialFunctionCall!(
   In Texture
   Random Color Texture
   In Texcoords=@n6.Out UV TexCoord
   Turn Time=Constant() {r=0.4}
   Distortion Tiling=Constant2Vector() {g=3, r=3}
   Distortion Amount=Constant() {r=0.035}
   Tiling
   Detail Factor
  ) {materialFunction=/Game/Effects/Niagara/MasterMaterials/Functions/Mojo_MF/MF_FX_TurnDistortion}
  n9 = LinearInterpolate(
   A=TextureSample(UVs=@n8.Out Base UV) {samplerType=SAMPLERTYPE_LinearGrayscale, texture=/SparksEffects/Effects/Intro/Rift_Portal/Textures/T_Plasma_Mask_A}.R
   B=TextureSample(UVs=@n8.Out Base UV) {samplerType=SAMPLERTYPE_LinearGrayscale, texture=/SparksEffects/Effects/Intro/Rift_Portal/Textures/T_Plasma_Mask_B}.R
   Alpha=Step(X=ParticleRandom()) {constX=0.5, constY=0.5}
  ) {constAlpha=0}
  n10 = DynamicParameter() {defaultValue=<rgba 0.1 1 1 1>, paramNames [Radial Reveal Value, Half or Full Plasma Mask Select, Erode Alpha, Param4]}
  n11 = LinearInterpolate(A=Multiply(A=Step(X=@n7) {constX=0.5, constY=0.25}, B=@n9), B=@n9, Alpha=@n10.Half or Full Plasma Mask Select)
  n12 = Saturate(None=Multiply(A=Power(Base=@n11) {constExponent=2.325}) {constB=3})
  MaterialFunctionCall(Emissive=Multiply(A=Multiply(A=@n5.RGB, B=@n12)) {constB=1.5, desc=Boost}) {materialFunction=&MF_FX_EmissiveAdjustments}
  n13 = Constant() {r=1}
  MaterialFunctionCall!(UV, Tile U=@n13, Tile V=@n13, Pan U=@n13, Pan V=@n13, Offset U, Offset V=Constant() {r=1}, Time=Multiply(A=Time()) {constB=0}) {materialFunction=/Game/Effects/Fort_Effects/Materials/_Functions/MF_PanTileOffset}
  n14 = Multiply(
   A=@n12
   B=MaterialFunctionCall!(Dynamic Param=@n10.Erode Alpha, Alpha=@n11) {materialFunction=/Game/Effects/Fort_Effects/Materials/_Functions/Erode_Alpha}.Result
  )
  n15 = SphereMask(A=@n2, B=Constant2Vector() {g=0.5, r=0.5}) {attenuationRadius=0.5, hardnessPercent=0.001}
  n16 = Sine(None=Multiply(A=@n15, B=@n10.Radial Reveal Value))
  n17 = Multiply(A=Multiply(A=Saturate(None=@n14), B=@n15), B=@n16) {desc=radial reveal edge to middle 1 to 0}
  n18 = LinearInterpolate(Alpha=MaterialFunctionCall!(Time Mult=Constant() {r=2}, Time Add=ParticleRandom()) {materialFunction=/Game/Effects/Niagara/MasterMaterials/Functions/Mojo_MF/MF_FX_Flicker}.Raw Flicker) {constA=0.9}
  Multiply(A=@n5.A, B=Saturate(None=Multiply(A=@n17, B=@n18)))
  Multiply(A=@n1, B=Constant2Vector() {g=2, r=-0.5})
  Multiply(A=MaterialFunctionCall!(Center, Maintain Aspect Ratio, Tiling=Constant2Vector() {g=1, r=0.5}, Offset, In Tex Coords) {materialFunction=/Game/Effects/Niagara/MasterMaterials/Functions/Mojo_MF/MF_FX_Uniform_Radial_UV}.Out UV)
  n19 = Constant()
  n20 = Constant() {r=0.75}
  Reroute(None=Add(
    A=MaterialFunctionCall!(UV=@n1, Perspective Rotation=@n19, Perspective Amount=@n20) {materialFunction=/FortUILibrary/MaterialFunctions/Basics/MF_UI_Perspective}.Perspective UVs
    B=Constant2Vector() {r=0.25}
   ))
  n21 = TextureCoordinate()
  n22 = MaterialFunctionCall!(UV=@n21, Twist Center, Twist Edges=Constant() {r=-0.1}) {materialFunction=/Game/Accessories/FORT_Backpacks/Backpack_F_MED_Celestial_Glow/FX/Materials/MF_PolarUV}
  Reroute(None=Panner(Coordinate=LinearInterpolate(A=@n21, B=@n22.Polar UVs) {constAlpha=0.9}) {bFractionalPart=true, speedY=1})
  Saturate(None=OneMinus(None=MaterialFunctionCall!(In Texcoords) {materialFunction=/Game/Effects/Niagara/MasterMaterials/Functions/Mojo_MF/MF_FX_Circle_SDF}.Result))
  n23 = MaterialFunctionCall(Time Mult=Constant() {r=0.8}, Time Add=ParticleRandom()) {materialFunction=&MF_FX_Flicker}
  Saturate(None=@n23.Raw Flicker)
  Step(X=@n23.Raw Flicker) {constY=0.2}
  Step(X=@n16) {constY=0.2}
  TextureCoordinate()
  TextureSample() {texture=/Game/Characters/Player/Female/Medium/Bodies/F_MED_LightningDragon/FX/T_LIghtning_Dragon_Inks_Long}
  TextureSample() {texture=/Game/Effects/Fort_Effects/Textures/T_FX_SpringerPlasma}
  TextureSample() {texture=/Game/Weapons/FORT_Melee/Pickaxe_Majesty_Male/FX/T_Lightning}
 ]
}
