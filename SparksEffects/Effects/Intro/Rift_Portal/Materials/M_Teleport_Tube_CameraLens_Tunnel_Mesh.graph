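# Translucent unlit Niagara mesh material for the Rift Portal intro's teleport tunnel, drawing animated energy streaks along a tube.
# Panning parallax-occluded streaks, a noise-distorted distant glow and rotated noise speed lines sum into emissive.
# UV gradients, particle alpha and depth fade shape opacity, while dynamic parameters add particle-color tint and fade to opaque white.
# A time-driven swirling vector, weighted by distance from an offset point, forms a wobble.
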Material /SparksEffects/Effects/Intro/Rift_Portal/Materials/M_Teleport_Tube_CameraLens_Tunnel_Mesh
graph {
 properties {
  bUsedWithNanite=true
  bUsedWithNiagaraMeshParticles=true
  blendMode=BLEND_Translucent
  pixelDepthOffsetMode=PDOM_Legacy
  shadingModel=MSM_Unlit
  usedShadingModels=Unlit
 }
 nodes [
  n1 = TextureCoordinate()
  n2 = ScalarParameter() {defaultValue=3, parameterName=Light_Tile_X}
  n3 = AppendVector(A=@n2, B=ScalarParameter() {defaultValue=0.5, parameterName=Base_Tile_Y})
  n4 = Multiply(A=ScalarParameter() {parameterName=Base_Pan_X}, B=Divide(A=@n2, B=ScalarParameter() {defaultValue=2, parameterName=Base_Tile_X}))
  n5 = ScalarParameter() {defaultValue=-0.85, parameterName=Base_Pan_Y}
  n6 = MaterialFunctionCall!(
   Heightmap Texture=TextureObject() {texture=/Game/Effects/Fort_Effects/Textures/Energy/T_OutpostShockwaveStreaks_02}
   Height Ratio=Constant() {r=0.01}
   Min Steps
   Max Steps
   UVs=Panner(Coordinate=Multiply(A=@n1, B=@n3), Speed=AppendVector(A=@n4, B=@n5))
   Heightmap Channel=Constant4Vector() {constant=<rgba 1 0 0 0>}.RGBA
   Reference Plane=Constant()
   -----------------------------
   Use World Coordinates
   Specify Manual Texture Size
   Manual Texture Size
   ------------------------------
   "Render Shadows (Occlusion Mapping)"
   Light Vector
   Shadow Steps
   Shadow Penumbra
   Transform To VertexNormal
  ) {materialFunction=/Engine/Functions/Engine_MaterialFunctions01/Texturing/ParallaxOcclusionMapping}
  n7 = Multiply(
   A=TextureSample(UVs=@n6.Parallax UVs) {texture=&T_OutpostShockwaveStreaks_02}.RGB
   B=Constant3Vector() {constant=<rgba 0.1096 0.5746 1 1>}
  )
  n8 = Multiply(
   A=TextureSample(UVs=@n6.Parallax UVs) {texture=&T_OutpostShockwaveStreaks_02}.RGB
   B=Constant3Vector() {constant=<rgba 0 1.1646 1.5 1>}
  )
  Add(A=Add(A=@n7, B=@n8), B=@n8) {constB=0}
  n9 = ComponentMask(None=TextureCoordinate()) {g=true}
  n10 = Clamp(None=Add(A=OneMinus(None=@n9)) {constB=-0.25})
  n11 = Constant() {r=3}
  n12 = TextureCoordinate()
  n13 = Constant() {r=1}
  n14 = Constant() {r=1}
  n15 = Panner(Coordinate=Multiply(A=TextureCoordinate(), B=AppendVector(A=@n13, B=@n14)), Speed=AppendVector(A=Constant() {r=0.5}, B=Constant() {r=-1}))
  n16 = Multiply(
   A=Power(
    Base=TextureSample(UVs=@n15) {texture=/Game/Characters/Player/Male/Medium/Bodies/M_MED_Viper/Textures/Glitch/T_SoftNoise}.R
    Exp=Constant() {r=1}
   )
   B=Constant() {r=0.2}
  )
  n17 = Multiply(
   A=Multiply(A=Power(Base=@n10, Exp=@n11), B=Constant() {r=3})
   B=TextureSample(UVs=Add(A=@n12, B=@n16)) {texture=&T_SoftNoise}.R
  )
  n18 = VectorParameter() {parameterName=DistantLightColor, defaultValue=<rgba 0.3 1.15 2 1>}
  n19 = MaterialFunctionCall!(
   UV Channel=TextureCoordinate()
   U Freq Multiplier=ScalarParameter() {defaultValue=2, group=UVs, parameterName=UV SpeedLines U Mult}
   V Freq Multiplier=ScalarParameter() {defaultValue=0.05, group=UVs, parameterName=UV SpeedLines V Mult}
   Pan Speed U=ScalarParameter() {defaultValue=0.4, group=UVs, parameterName=UV SpeedLines U Pan Speed}
   Pan Speed V=ScalarParameter() {defaultValue=-0.25, group=UVs, parameterName=UV SpeedLines V Pan Speed}
  ) {materialFunction=/Game/Effects/Niagara/MasterMaterials/Functions/UVs/MF_FX_TextureUVPanner}
  n20 = Constant() {r=4.7125}
  n21 = Power(
   Base=TextureSample(UVs=Rotator(Coordinate=@n19.Result, Time=@n20) {speed=1}) {samplerSource=SSM_Wrap_WorldGroupSettings, texture=/Game/Global/Textures/Default/Noise/Noise}.G
   Exp=ScalarParameter() {defaultValue=12, parameterName=Speed Lines Power}
  ) {constExponent=18}
  n22 = VectorParameter() {parameterName=Speed Lines Color, defaultValue=<rgba 0 0.7731 1 1>}
  n23 = Add(A=Add(A=Multiply(A=@n17, B=@n18.RGB), B=Multiply(A=@n21, B=@n22.RGB) {constB=0.25}), B=@n7)
  n24 = ParticleColor()
  Add(A=@n23, B=@n24.RGB)
  Add(A=@n23)
  n25 = Multiply(A=Time(), B=ScalarParameter() {defaultValue=-0.15, parameterName=RotationSpeed})
  n26 = Add(A=@n25, B=Sine(None=Multiply(A=@n25) {constB=0.45}))
  n27 = MaterialFunctionCall(R=Sine(None=@n26), B=Cosine(None=@n26)) {materialFunction=&AppendMany}
  n28 = ScalarParameter() {defaultValue=7, parameterName=Multiplier}
  n29 = WorldPosition() {worldPositionShaderOffset=WPT_ExcludeAllShaderOffsets}
  n30 = TransformPosition(None=VectorParameter() {parameterName=Offset, defaultValue=<rgba 0 0 0 1>}.RGB) {transformType=TRANSFORMPOSSOURCE_World}
  n31 = Saturate(None=Divide(A=Distance(A=@n29.XYZ, B=@n30), B=ScalarParameter() {defaultValue=150, parameterName=Rad}))
  LinearInterpolate(
   A=Constant3Vector()
   B=Transform(None=Multiply(A=@n27.RGB, B=@n28)) {transformSourceType=TRANSFORMSOURCE_Local}
   Alpha=Power(Base=OneMinus(None=@n31), Exp=ScalarParameter() {defaultValue=2, parameterName=LinearFalloffPower})
  )
  n32 = DynamicParameter() {parameterIndex=2, defaultValue=<rgba 1 0 1 1>, paramNames [Param1, White Fade, Param3, Param4]}
  n33 = TextureCoordinate()
  n34 = Power(Base=ComponentMask(None=@n33) {g=true}) {constExponent=6}
  n35 = Multiply(A=Saturate(None=@n32.Param1), B=Saturate(None=Multiply(A=@n34) {constB=1.25}))
  n36 = LinearInterpolate(B=@n24.RGB, Alpha=@n35)
  n37 = Saturate(None=@n32.White Fade)
  MaterialFunctionCall(Emissive=LinearInterpolate(A=Add(A=@n23, B=@n36), B=Constant3Vector() {constant=<rgba 1 1 1 0>}, Alpha=@n37)) {materialFunction=&MF_FX_EmissiveAdjustments}
  OneMinus(None=@n33)
  Reroute(None=@n23)
  n38 = Power(Base=ComponentMask(None=TextureCoordinate()) {g=true}) {constExponent=3.5}
  n39 = Multiply(A=Saturate(None=Multiply(A=@n38) {constB=2}), B=Desaturation(None=@n23))
  n40 = Multiply(A=Multiply(A=Add(A=@n39, B=@n35), B=@n24.A), B=DepthFade(FadeDistance=ScalarParameter() {defaultValue=5, parameterName=Depth Fade Distance}))
  Reroute(None=LinearInterpolate(A=Saturate(None=@n40), Alpha=@n37))
 ]
}
