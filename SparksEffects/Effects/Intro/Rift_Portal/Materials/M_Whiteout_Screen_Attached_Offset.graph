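# Translucent unlit Niagara whiteout material for the Rift Portal intro that ignores depth testing and attaches its geometry to the camera.
# An offset computed from camera minus actor position moves it onto the camera, then throws it far away beyond a fade distance or when facing away from an offset anchor point.
# Particle color feeds an emissive adjustment function, while dynamic parameters scale the fade distance and supply the whiteout alpha.
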Material /SparksEffects/Effects/Intro/Rift_Portal/Materials/M_Whiteout_Screen_Attached_Offset
graph {
 properties {
  bDisableDepthTest=true
  bUsedWithNiagaraMeshParticles=true
  bUsedWithNiagaraSprites=true
  blendMode=BLEND_Translucent
  pixelDepthOffsetMode=PDOM_Legacy
  shadingModel=MSM_Unlit
  twoSided=true
  usedShadingModels=Unlit
 }
 nodes [
  n1 = WorldPosition()
  n2 = Add(A=CameraPositionWS(), B=@n1.XYZ)
  n3 = Multiply(A=CameraPositionWS(), B=VectorParameter() {parameterName=OffsetMultiplier, defaultValue=<rgba 1 1 1 0>}.RGB)
  n4 = Add(
   A=Multiply(A=ActorPositionWS(), B=VectorParameter() {parameterName=OffsetMultiplier, defaultValue=<rgba 1 1 1 0>}.RGB)
   B=VectorParameter() {parameterName=Offset, defaultValue=<rgba 0 0 5000 0>}.RGB
  )
  n5 = DynamicParameter() {paramNames [FadeOutDistanceMult, Whiteout, Param3, Param4]}
  n6 = Divide(A=Distance(A=@n3, B=@n4), B=Multiply(A=@n5.FadeOutDistanceMult, B=ScalarParameter() {defaultValue=150000, parameterName=FadeDistance}))
  n7 = Floor(None=Multiply(A=Saturate(None=@n6)) {constB=1.01})
  n8 = Normalize(VectorInput=Subtract(A=@n3, B=@n4))
  n9 = Transform(None=Constant3Vector() {constant=<rgba 0 0 1 0>}) {transformSourceType=TRANSFORMSOURCE_Camera}
  n10 = Multiply(A=OneMinus(None=DotProduct(A=@n8, B=@n9)), B=ScalarParameter() {defaultValue=3, parameterName=PopOutThreshold}) {constB=2}
  n11 = OneMinus(None=Saturate(None=Floor(None=@n10)))
  Add(
   A=Subtract(A=Subtract(A=@n2, B=@n1.XYZ), B=ActorPositionWS())
   B=Multiply(A=Add(A=@n7, B=@n11), B=Constant3Vector() {constant=<rgba 0 0 -54654566400 0>})
  )
  Constant() {r=1}
  LinearInterpolate(Alpha=Saturate(None=@n5.Whiteout))
  MaterialFunctionCall(Emissive=ParticleColor().RGB) {materialFunction=&MF_FX_EmissiveAdjustments}
  NamedRerouteDeclaration(None=@n5.Whiteout) {name=Alpha, nodeColor=<rgba 1 0 0.6797 1>}
  Saturate(None=NamedRerouteUsage())
 ]
}
